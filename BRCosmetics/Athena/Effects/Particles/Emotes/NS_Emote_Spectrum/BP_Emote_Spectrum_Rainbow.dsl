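; Blueprint actor for the rainbow mesh effect of the Spectrum emote, handling its fade-in and its handoff to a fade-out.
; On BeginPlay it writes game time and fade-in duration into the mesh's custom primitive data, presumably for a material-driven fade, and applies a negative translucency sort distance offset.
; When destroyed, it spawns a separate FadeOut actor at its transform, so the rainbow can outlive the original actor.
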(event EventBeginPlay
  (bind _rainbowmesh (Variables|Default|GetRainbowMesh))
  (Rendering|Material|SetCustomPrimitiveDataFloat _rainbowmesh 0 (Utilities|Time|GetGameTimeinSeconds))
  (Rendering|Material|SetCustomPrimitiveDataFloat _rainbowmesh 1 (Variables|Default|GetFadeInDuration))
  (Rendering|SetTranslucencySortDistanceOffset (Variables|Default|GetRainbowMesh) -5e+01)
  (Game|BindEventtoOnDestroyed (delegate Custom|RainbowDeath)))

(event Custom|RainbowDeath
  (params (DestroyedActor :type Actor%20Object%20Reference))
  (Game|SpawnActorfromClass "/BRCosmetics/Athena/Effects/Particles/Emotes/NS_Emote_Spectrum/BP_Emote_Spectrum_Rainbow_FadeOut.BP_Emote_Spectrum_Rainbow_FadeOut_C" (Transformation|GetActorTransform) "Undefined" "MultiplyWithRoot" 0 0))

(fn ConstructionScript)
