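; Blueprint static mesh actor for the rainbow fade-out effect of the Spectrum emote.
; On BeginPlay it writes the current game time and the fade-out duration into the mesh's custom primitive data slots, presumably for the material to drive the fade.
; It then sets its own life span to that duration so the actor destroys itself when the fade ends.
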(event EventBeginPlay
  (bind _staticmeshcomponent (Variables|StaticMeshActor|GetStaticMeshComponent))
  (bind _fadeoutduration (Variables|Default|GetFadeOutDuration))
  (Rendering|Material|SetCustomPrimitiveDataFloat _staticmeshcomponent 0 (Utilities|Time|GetGameTimeinSeconds))
  (Rendering|Material|SetCustomPrimitiveDataFloat _staticmeshcomponent 1 _fadeoutduration)
  (Actor|SetLifeSpan _fadeoutduration))

(fn ConstructionScript)
