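# Additive unlit Niagara sprite material for the beam in the Spectrum emote effect.
# It shapes the beam from UVs: a soft horizontal falloff that narrows toward the bottom, times a reflected vertical gradient, sharpened by an exponent.
# The mask ramps up with particle relative time, then gets depth fade, a color tint and shared emissive adjustments.
# A separate branch computes a camera-facing push offset from world position.
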Material /BRCosmetics/Athena/Effects/Particles/Emotes/NS_Emote_Spectrum/M_Emote_Spectrum_Beam
graph {
 properties {bUsedWithNiagaraSprites=true, blendMode=BLEND_Additive, pixelDepthOffsetMode=PDOM_Legacy, shadingModel=MSM_Unlit, usedShadingModels=Unlit}
 nodes [
  n1 = TextureCoordinate()
  n2 = ComponentMask(None=@n1) {g=true}
  n3 = ScalarParameter() {defaultValue=0.7, parameterName=Bottom Squeeze, sliderMax=1}
  n4 = Divide(A=Subtract(A=ComponentMask(None=@n1) {r=true}) {constB=0.5}, B=OneMinus(None=Multiply(A=@n2, B=@n3)))
  n5 = ConstantBiasScale(None=Saturate(None=Add(A=@n4) {constB=0.5})) {bias=-0.5, scale=2}
  n6 = OneMinus(None=DotProduct(A=@n5, B=@n5))
  n7 = MaterialFunctionCall!(Gradient=@n2, Center Point=ScalarParameter() {defaultValue=0.85, parameterName=Center Point}) {materialFunction=/Game/Effects/Fort_Effects/MaterialFunctions/MF_GradientReflected}
  n8 = Multiply(
   A=Power(Base=Multiply(A=@n6, B=@n7.Result), Exp=ScalarParameter() {defaultValue=2, parameterName=Exponent, sliderMax=9, sliderMin=1})
   B=Saturate(None=ParticleRelativeTime())
  )
  n9 = ScalarParameter() {defaultValue=8, parameterName=Depth Fade Distance, sliderMax=33}
  MaterialFunctionCall(Emissive=Multiply(A=VectorParameter() {parameterName=Color, defaultValue=<rgba 0.5 0.38 0.2385 0>}.RGB, B=DepthFade(Opacity=@n8, FadeDistance=@n9))) {materialFunction=&MF_FX_EmissiveAdjustments}
  n10 = CameraPositionWS()
  n11 = WorldPosition()
  Multiply(A=Normalize(VectorInput=Subtract(A=@n10, B=@n11.XYZ)), B=ScalarParameter() {defaultValue=5, parameterName=Push})
 ]
}
