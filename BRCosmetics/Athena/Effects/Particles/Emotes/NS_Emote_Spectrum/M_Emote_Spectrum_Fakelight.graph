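# Additive unlit Niagara sprite material that fakes a rainbow point light on scene surfaces for the Spectrum emote.
# Distance from the world position behind translucency to the particle center, over particle radius, forms a spherical falloff.
# That falloff plus layered sine waves of time, particle random and UVs scrolls a shared rainbow gradient, blended toward a reveal color by powered falloff and scaled by brightness.
# It also computes a radius-length vector toward the camera.
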Material /BRCosmetics/Athena/Effects/Particles/Emotes/NS_Emote_Spectrum/M_Emote_Spectrum_Fakelight
graph {
 properties {
  bUsedWithNiagaraSprites=true
  blendMode=BLEND_Additive
  pixelDepthOffsetMode=PDOM_Legacy
  shadingModel=MSM_Unlit
  translucencyPass=MTP_BeforeDOF
  usedShadingModels=Unlit
 }
 nodes [
  n1 = Constant() {desc=Time bumpiness, r=6}
  n2 = Add(A=Time(), B=ParticleRandom())
  n3 = Divide() {constA=355, constB=113, desc=Pi}
  n4 = Divide(B=@n3)
  n5 = Add(A=@n2, B=DotProduct(A=TextureCoordinate(), B=Constant() {r=2}))
  n6 = Add(A=Multiply(A=@n1, B=@n2) {constA=6}, B=Sine(None=Multiply(A=@n4, B=@n5)))
  n7 = Divide(B=@n3) {constA=1.618}
  n8 = Multiply(A=Sine(None=Multiply(A=@n7, B=@n5))) {constB=0.618}
  n9 = Sine(None=Multiply(A=Divide(B=@n3) {constA=4.235}, B=@n5))
  n10 = Multiply(A=Add(A=Add(A=@n6, B=@n8), B=Multiply(A=@n9) {constB=0.236}), B=Divide(B=@n1)) {constB=0.1667}
  n11 = Distance(
   A=MaterialFunctionCall() {materialFunction=&WorldPositionBehindTranslucency}.Result
   B=ParticlePositionWS()
  )
  n12 = ParticleRadius()
  n13 = OneMinus(None=Saturate(None=Divide(A=@n11, B=@n12)))
  n14 = MaterialFunctionCall!(Gradient=Frac(None=Add(A=@n10, B=@n13))) {materialFunction=/BRCosmetics/Athena/Effects/Particles/Skydiving/NS_Contrail_DarkAndLight/MF_DarkAndLight_RainbowColors}
  n15 = VectorParameter() {parameterName=Rainbow Reveal Color, defaultValue=<rgba 3 1.3305 1.11 1>}
  n16 = AppendVector(A=@n15.RGB, B=@n15.RGB)
  n17 = Power(Base=@n13, Exp=ScalarParameter() {defaultValue=2, parameterName=Falloff, sliderMax=6}) {constExponent=3}
  MaterialFunctionCall(Emissive=Multiply(A=LinearInterpolate(A=@n14.Result, B=@n16, Alpha=@n17), B=ScalarParameter() {defaultValue=0.25, parameterName=Brightness}) {constB=5}) {materialFunction=&MF_FX_EmissiveAdjustments}
  n18 = CameraPositionWS()
  n19 = WorldPosition()
  Multiply(A=@n12, B=Normalize(VectorInput=Subtract(A=@n18, B=@n19.XYZ)))
 ]
}
