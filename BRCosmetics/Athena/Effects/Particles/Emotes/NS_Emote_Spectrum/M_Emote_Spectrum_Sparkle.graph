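# Unlit alpha-composite Niagara sprite material for the rainbow sparkle particles of the Spectrum emote effect.
# Each particle picks a random quantized rainbow color, blended with particle color over its life.
# The procedural star shape combines centered UVs with a 45-degree rotated copy, using absolute-value powers.
# Randomly offset, layered sines with irrational-like frequencies flicker the shape's exponent and sharpness, with depth fade softening intersections.
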Material /BRCosmetics/Athena/Effects/Particles/Emotes/NS_Emote_Spectrum/M_Emote_Spectrum_Sparkle
graph {
 properties {
  bUsedWithNiagaraSprites=true
  blendMode=BLEND_AlphaComposite
  pixelDepthOffsetMode=PDOM_Legacy
  shadingModel=MSM_Unlit
  usedShadingModels=Unlit
 }
 nodes [
  n1 = ParticleColor()
  n2 = ParticleRandom()
  n3 = ScalarParameter() {defaultValue=6, parameterName=Steps}
  n4 = Subtract(
   A=OneMinus(None=MaterialFunctionCall!(Gradient=@n2, Steps=@n3) {materialFunction=/Game/Effects/Fort_Effects/MaterialFunctions/MF_Quantize}.Default Range)
   B=Divide(B=@n3)
  )
  n5 = ComponentMask(None=MaterialFunctionCall!(Gradient=Frac(None=@n4)) {materialFunction=/BRCosmetics/Athena/Effects/Particles/Skydiving/NS_Contrail_DarkAndLight/MF_DarkAndLight_RainbowColors}.Result) {b=true, g=true, r=true}
  n6 = Constant() {r=0.2}
  n7 = Constant() {desc=Time bumpiness, r=6}
  n8 = Time()
  n9 = Divide() {constA=355, constB=113, desc=Pi}
  n10 = Divide(B=@n9)
  n11 = Add(A=@n8, B=Multiply(A=@n7, B=ParticleRandom()) {constB=9.876})
  n12 = Add(A=Multiply(A=@n7, B=@n8) {constA=6}, B=Sine(None=Multiply(A=@n10, B=@n11)))
  n13 = Divide(B=@n9) {constA=1.618}
  n14 = Multiply(A=Sine(None=Multiply(A=@n13, B=@n11))) {constB=0.618}
  n15 = Sine(None=Multiply(A=Divide(B=@n9) {constA=4.235}, B=@n11))
  n16 = Multiply(A=Add(A=Add(A=@n12, B=@n14), B=Multiply(A=@n15) {constB=0.236}), B=Divide(B=@n7)) {constB=0.1667}
  MaterialFunctionCall(Hue Shift Percentage=@n16, Texture=Multiply(A=@n1.RGB, B=Desaturation(None=@n5, Fraction=@n6))) {materialFunction=&HueShift}
  n17 = Power(Base=ParticleRelativeTime()) {constExponent=0.3}
  n18 = Divide(A=Subtract(A=TextureCoordinate()) {constB=0.5})
  n19 = ComponentMask(None=@n18) {r=true}
  n20 = Constant() {r=0.125}
  n21 = Cosine(None=@n20)
  n22 = Multiply(A=@n19, B=@n21)
  n23 = ComponentMask(None=@n18) {g=true}
  n24 = Sine(None=@n20)
  n25 = Multiply(A=@n23, B=@n24)
  n26 = Multiply(A=@n21, B=@n23)
  n27 = Multiply(A=@n24, B=@n19)
  n28 = AppendVector(A=@n18, B=AppendVector(A=Add(A=@n22, B=@n25), B=Subtract(A=@n26, B=@n27)))
  n29 = Constant() {desc=Flicker speed, r=1.618}
  n30 = ConstantBiasScale(None=Sine(None=Multiply(A=@n16, B=@n29)))
  n31 = VertexInterpolator(VS=AppendVector(A=LinearInterpolate(Alpha=@n30) {constA=0.3, constB=0.4}, B=LinearInterpolate(Alpha=@n30) {constA=5, constB=6}))
  n32 = Power(Base=Abs(None=VertexInterpolator(VS=@n28).PS), Exp=ComponentMask(None=@n31.PS) {r=true})
  n33 = Saturate(None=MaterialFunctionCall(f2=ComponentMask(None=@n32) {g=true, r=true}) {materialFunction=&AddComponents}.f2)
  n34 = OneMinus(None=Divide(A=OneMinus(None=@n33)) {constB=3})
  n35 = Saturate(None=MaterialFunctionCall(f2=ComponentMask(None=@n32) {a=true, b=true}) {materialFunction=&AddComponents}.f2)
  n36 = OneMinus(None=Saturate(None=Multiply(A=@n34, B=@n35)))
  n37 = ParticleColor()
  n38 = ParticleSize()
  n39 = Divide(A=ComponentMask(None=@n31.PS) {g=true}, B=DepthFade(Opacity=@n37.A, FadeDistance=ComponentMask(None=@n38) {r=true}))
  MaterialFunctionCall(Emissive=Multiply(A=LinearInterpolate(A=@n1.RGB, B=@n5, Alpha=@n17) {constAlpha=0.9}, B=Power(Base=@n36, Exp=@n39))) {materialFunction=&MF_FX_EmissiveAdjustments}
  n40 = Divide(A=ParticleRelativeTime())
  Multiply(A=@n37.A, B=SquareRoot(None=Sine(None=@n40)))
  Time()
 ]
}
