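# Additive unlit sparkle material for the Spectrum emote effect that scatters and fades out mesh sparkles as rainbow star sprites.
# The emote timeline function, randomized by UV1, drives an outward offset, brightness fade and desaturation.
# The Sprite function rebuilds quads from UV2, with time-based shimmer scaling their size.
# Color comes from a mirrored, stepped rainbow gradient, while two rotated power-curve crosses form a flickering star mask with depth fade.
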Material /BRCosmetics/Athena/Effects/Particles/Emotes/NS_Emote_Spectrum/M_Emote_Spectrum_Sparkles_Mesh_FadeOut
graph {
 properties {
  bDisableDepthTest=true
  blendMode=BLEND_Additive
  pixelDepthOffsetMode=PDOM_Legacy
  shadingModel=MSM_Unlit
  twoSided=true
  usedShadingModels=Unlit
 }
 nodes [
  Frac(None=Time())
  n1 = Subtract(A=TextureCoordinate() {coordinateIndex=1}) {constB=0.5}
  n2 = ScalarParameter() {defaultValue=1, parameterName=Offset Randomness, sliderMax=1}
  n3 = Multiply(A=DeriveNormalZ(InXY=Multiply(A=@n1, B=@n2)), B=ScalarParameter() {defaultValue=33, parameterName=Offset Distance})
  n4 = LinearInterpolate(A=Transform(None=@n3), B=@n3)
  n5 = MaterialFunctionCall() {materialFunction=/Game/Effects/Fort_Effects/MaterialFunctions/MF_Emote_Spectrum_Timeline}
  n6 = ComponentMask(None=TextureCoordinate() {coordinateIndex=1}) {r=true}
  n7 = ScalarParameter() {defaultValue=3, parameterName=Lifetime Randomization Exponent, sliderMax=1}
  n8 = Saturate(None=Divide(A=OneMinus(None=@n5.Result), B=Power(Base=@n6, Exp=@n7)))
  n9 = SquareRoot(None=@n8)
  n10 = TextureCoordinate()
  n11 = Constant() {r=0.5}
  n12 = TextureCoordinate() {coordinateIndex=1}
  n13 = Add(
   A=Add(A=DotProduct(A=@n10, B=@n11), B=DotProduct(A=@n12, B=@n11))
   B=Multiply(A=Time(), B=ScalarParameter() {defaultValue=0.2, parameterName=Shimmer Speed})
  )
  n14 = Sine(None=Divide(A=Frac(None=@n13)))
  n15 = ScalarParameter() {defaultValue=333, parameterName=Sparkle Power}
  n16 = ScalarParameter() {defaultValue=6, parameterName=Sprite Scale}
  n17 = Power(Base=@n8) {constExponent=0.333}
  n18 = Divide(A=Subtract(A=TextureCoordinate() {coordinateIndex=2}) {constB=0.5})
  n19 = ComponentMask(None=@n18) {r=true}
  n20 = ScalarParameter() {defaultValue=0.125, parameterName=Sprite Rotation}
  n21 = Cosine(None=@n20)
  n22 = ComponentMask(None=@n18) {g=true}
  n23 = Sine(None=@n20)
  n24 = Add(A=Multiply(A=@n19, B=@n21), B=Multiply(A=@n22, B=@n23))
  n25 = Subtract(A=Multiply(A=@n21, B=@n22), B=Multiply(A=@n23, B=@n19))
  MaterialFunctionCall!(
   Center Location=Add(A=WorldPosition().XYZ, B=Multiply(A=@n4, B=@n9)) {desc=adding offset}
   0-1 UVs=Add(A=AppendVector(A=@n24, B=@n25)) {constB=0.5}
   XY scale=Multiply(A=Power(Base=@n14, Exp=@n15) {constExponent=33}, B=Divide(A=@n16, B=@n17))
   OverridePivot
   "Normals (Optional)"
  ) {materialFunction=&Sprite}
  n26 = OneMinus(None=@n8)
  n27 = MaterialFunctionCall!(
   Gradient=Saturate(None=MaterialFunctionCall() {materialFunction=/Game/Effects/Fort_Effects/MaterialFunctions/MF_U}.U)
   Center Point=Constant() {r=0.5}
  ) {materialFunction=/Game/Effects/Fort_Effects/MaterialFunctions/MF_GradientReflected}
  n28 = ScalarParameter() {defaultValue=6, parameterName=Steps}
  n29 = Subtract(
   A=OneMinus(None=MaterialFunctionCall!(Gradient=@n27.Result, Steps=@n28) {materialFunction=/Game/Effects/Fort_Effects/MaterialFunctions/MF_Quantize}.Default Range)
   B=Divide(B=@n28)
  )
  n30 = ComponentMask(None=MaterialFunctionCall!(Gradient=Frac(None=@n29)) {materialFunction=/BRCosmetics/Athena/Effects/Particles/Skydiving/NS_Contrail_DarkAndLight/MF_DarkAndLight_RainbowColors}.Result) {b=true, g=true, r=true}
  n31 = ScalarParameter() {defaultValue=7, parameterName=Rainbow Brightness}
  n32 = SquareRoot(None=@n8)
  n33 = ScalarParameter() {defaultValue=0.3, parameterName=Rainbow Desaturate, sliderMax=1}
  Multiply(A=Multiply(A=@n26, B=@n26), B=Desaturation(None=Multiply(A=@n30, B=@n31), Fraction=Max(A=@n32, B=@n33)))
  n34 = Divide(A=Subtract(A=TextureCoordinate() {coordinateIndex=2}) {constB=0.5})
  n35 = ComponentMask(None=@n34) {r=true}
  n36 = Constant() {r=0.125}
  n37 = Cosine(None=@n36)
  n38 = Multiply(A=@n35, B=@n37)
  n39 = ComponentMask(None=@n34) {g=true}
  n40 = Sine(None=@n36)
  n41 = Multiply(A=@n39, B=@n40)
  n42 = Multiply(A=@n37, B=@n39)
  n43 = Multiply(A=@n40, B=@n35)
  n44 = AppendVector(A=@n34, B=AppendVector(A=Add(A=@n38, B=@n41), B=Subtract(A=@n42, B=@n43)))
  n45 = Constant() {desc=Time bumpiness, r=6}
  n46 = Time()
  n47 = Divide() {constA=355, constB=113, desc=Pi}
  n48 = Divide(B=@n47)
  n49 = TextureCoordinate() {coordinateIndex=1}
  n50 = Add(A=@n46, B=Multiply(A=@n45, B=ComponentMask(None=@n49) {r=true}) {constB=9.876})
  n51 = Add(A=Multiply(A=@n45, B=@n46) {constA=6}, B=Sine(None=Multiply(A=@n48, B=@n50)))
  n52 = Divide(B=@n47) {constA=1.618}
  n53 = Multiply(A=Sine(None=Multiply(A=@n52, B=@n50))) {constB=0.618}
  n54 = Sine(None=Multiply(A=Divide(B=@n47) {constA=4.235}, B=@n50))
  n55 = Multiply(A=Add(A=Add(A=@n51, B=@n53), B=Multiply(A=@n54) {constB=0.236}), B=Divide(B=@n45)) {constB=0.1667}
  n56 = Constant() {desc=Flicker speed, r=1.618}
  n57 = ConstantBiasScale(None=Sine(None=Multiply(A=@n55, B=@n56)))
  n58 = VertexInterpolator(VS=AppendVector(A=LinearInterpolate(Alpha=@n57) {constA=0.3, constB=0.4}, B=LinearInterpolate(Alpha=@n57) {constA=2, constB=3}))
  n59 = Power(Base=Abs(None=VertexInterpolator(VS=@n44).PS), Exp=ComponentMask(None=@n58.PS) {r=true})
  n60 = ComponentMask(None=@n59) {g=true, r=true}
  n61 = OneMinus(None=Saturate(None=MaterialFunctionCall(f2=@n60) {materialFunction=&AddComponents}.f2))
  n62 = ComponentMask(None=@n59) {a=true, b=true}
  n63 = Multiply(
   A=OneMinus(None=Divide(A=@n61) {constB=3})
   B=Saturate(None=MaterialFunctionCall(f2=@n62) {materialFunction=&AddComponents}.f2)
  )
  n64 = ScalarParameter() {defaultValue=1, parameterName=Opacity, sliderMax=1}
  n65 = ScalarParameter() {defaultValue=7, parameterName=Depth Fade Distance, sliderMax=33}
  Power(Base=OneMinus(None=Saturate(None=@n63)), Exp=Divide(A=ComponentMask(None=@n58.PS) {g=true}, B=DepthFade(Opacity=@n64, FadeDistance=@n65)))
 ]
}
