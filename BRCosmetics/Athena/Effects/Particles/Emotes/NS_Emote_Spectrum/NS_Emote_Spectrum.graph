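# Niagara system for the Spectrum emote, pairing a fake-light glow with sparkles apparently meant for a hand.
# FakeLight bursts one persistent particle on two sprite renderers, growing it from zero while its alignment blends from opposite owner velocity to world up.
# SparklesHand spawns short-lived bright sprites in a small sphere with cone velocity.
# Emitter-age curves ramp sparkle spawn probability and upward acceleration, so the effect builds gradually.
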NiagaraSystem /BRCosmetics/Athena/Effects/Particles/Emotes/NS_Emote_Spectrum/NS_Emote_Spectrum
dependencies {
 usedDataInterfaces [&NiagaraDataInterfaceCurve, &NiagaraDataInterfaceVectorCurve]
 usedDynamicInputs []
 usedModules [&SystemState]
}
emitters {
 FakeLight {
  emitterUpdateScript [
   &EmitterState(Life Cycle Mode=::NewEnumerator1, Loop Behavior=::NewEnumerator1, Loop Duration Mode=::NewEnumerator1, UseLoopDelay=false)
   &SpawnBurst_Instantaneous(Spawn Count=1, Use Spawn Probability=false)
  ]
  particleSpawnScript [&InitializeParticle(
    Color Mode=::NewEnumerator0
    Lifetime=2
    Sprite Size Mode=::NewEnumerator3
    Uniform Sprite Size=22
    UsePositionOffset=false
    Write Material Random=true
   )]
  particleUpdateScript [
   &ParticleState(Kill Particles When Lifetime Has Elapsed=false, Let Infinitely Lived Particles Die When Emitter Deactivates=true)
   &ScaleSpriteSize(Scale Factor=&Vector2DFromFloat(Value=&FloatFromCurve(FloatCurve=DataInterface() {propertyValues {Curve {keys [<at 0 0>,<at 1 1>]}}})))
   /NiagaraFN/Sprite/FN_SpriteAlignmentAndFacing(SpriteAlignment:Vector3f=&Lerp_Vector(
     A=&NormalizeVector(
      Vector To Normalize=&Multiply_VectorByFloat(Vector=$Engine.Owner.Velocity, Float=-1)
      Safe Normalize=true
      Safe Fallback Threshold=0
      Safe Fallback Vector=<0 0 0.001>
     )
     B=<0 0 1>
     Alpha=&ClampFloat(Float=$Particles.NormalizedAge)
    ))
  ]
  rendererClasses [&NiagaraSpriteRendererProperties]
  renderers [
   {rendererClass=&NiagaraSpriteRendererProperties}
   {rendererClass=&NiagaraSpriteRendererProperties, rendererIndex=1}
  ]
 }
 SparklesHand {
  emitterUpdateScript [
   &EmitterState()
   &SpawnRate(
    Spawn Probability=&FloatFromCurve(FloatCurve=DataInterface() {propertyValues {Curve {keys [<at 0.5 0>,<at 3 1>]}}}, CurveIndex=$Emitter.Age)
    SpawnRate=11
    Use Spawn Probability=true
   )
  ]
  particleSpawnScript [
   &InitializeParticle(
    Color=<rgba 13 9.2105 8.71 1>
    Lifetime Max=0.666
    Lifetime Min=0.333
    Lifetime Mode=::NewEnumerator1
    Sprite Rotation Angle=0.125
    Sprite Rotation Mode=::NewEnumerator1
    Sprite Size Mode=::NewEnumerator1
    UsePositionOffset=false
    Write Material Random=true
   )
   &AddVelocityInCone(
    Cone Angle=90
    Module.Use Velocity Falloff On Cone Axis=false
    Velocity Distribution Along Cone Axis=0.9
    Velocity Strength=&RandomRangeFloat(Minimum=11, Maximum=22)
   )
   &SphereLocation(
    MaskSpawnGroup=false
    Override Seed=false
    Sphere Radius=4
    Surface Only=false
    UseNonUniformScale=false
    UseOffset=false
    UseSphereOrientation=false
   )
  ]
  particleUpdateScript [
   &AccelerationForce(Acceleration=/Niagara/DynamicInputs/ScaleByCurve/ScaleVectorByCurve(
     InitialValue:Vector3f=<0 0 11>
     ScaleCurve:NiagaraDataInterfaceVectorCurve=DataInterface() {propertyValues {XCurve {keys []}, YCurve {keys []}, ZCurve {keys [<at 0 0>,<at 2 1>]}}}
     CurveIndex:NiagaraFloat=$Emitter.Age
    ))
   &SolveForcesAndVelocity(Clamp Velocity=false, Limit Acceleration=false)
   &ParticleState()
   &ScaleSpriteSize(Scale Factor=&Vector2DFromFloat(Value=&FloatFromCurve(FloatCurve=DataInterface() {propertyValues {Curve {keys [<at 0 0>,<at 0.3 1>,<at 0.6 1.075>,<at 1 0>]}}})))
   &CameraOffset(Camera Offset Amount=3)
  ]
  rendererClasses [&NiagaraSpriteRendererProperties]
  renderers [{rendererClass=&NiagaraSpriteRendererProperties}]
 }
}
summary {
 systemName=NS_Emote_Spectrum
 emitters [
  {emitterName=FakeLight, rendererClasses [&NiagaraSpriteRendererProperties]}
  {emitterName=SparklesHand, rendererClasses [&NiagaraSpriteRendererProperties]}
 ]
 userVariables [
  User.bIsFrontEnd {type=NiagaraBool}
  User.bIsFrontEndPreview {type=NiagaraBool}
  User.PlayerSpeed {type=NiagaraFloat, defaultValue=0, description=""}
  User.SkeletalMesh {
   type=NiagaraDataInterfaceSkeletalMesh
   defaultValue {
    struct=NiagaraExt_VariableValue_DataInterface
    value {dataInterface=&self:NiagaraDataInterfaceSkeletalMesh_0}
   }
  }
 ]
}
